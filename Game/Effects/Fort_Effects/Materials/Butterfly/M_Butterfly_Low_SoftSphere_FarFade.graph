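# Unlit additive mesh-particle material in the Butterfly effects set, drawing a soft, Fresnel-shaped glow, likely on a sphere as the name suggests.
# Particle color is multiplied by MF_Fresnel_FX, with brightness and exponent parameters, and by Near_Camera_Fade over a parameterized distance.
# The default shading path applies a depth fade to soften intersections, while mobile skips it.
# The result feeds MF_FX_EmissiveAdjustments as emissive.
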Material /Game/Effects/Fort_Effects/Materials/Butterfly/M_Butterfly_Low_SoftSphere_FarFade
graph {
 properties {bUsedWithMeshParticles=true, blendMode=BLEND_Additive, pixelDepthOffsetMode=PDOM_Legacy, shadingModel=MSM_Unlit, usedShadingModels=Unlit}
 nodes [
  n1 = ScalarParameter() {defaultValue=1, parameterName=Fresnel Brightness}
  n2 = ScalarParameter() {defaultValue=10, parameterName=Fresnel Exponent}
  n3 = Multiply(
   A=Multiply(
    A=ParticleColor().RGB
    B=MaterialFunctionCall(Brightness=@n1, Exponent=@n2) {materialFunction=&MF_Fresnel_FX}.Fresnel
   )
   B=MaterialFunctionCall(Distance=ScalarParameter() {defaultValue=2500, parameterName=CameraFadeDistance}) {materialFunction=&Near_Camera_Fade}.Result
  )
  n4 = ScalarParameter() {defaultValue=5000, parameterName=Depth Fade Distance}
  MaterialFunctionCall(Emissive=ShadingPathSwitch(Default=DepthFade(Opacity=@n3, FadeDistance=@n4), Mobile=@n3) {desc=Migrated from FeatureLevelSwitch}) {materialFunction=&MF_FX_EmissiveAdjustments}
 ]
}
